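; Behavior tree decorator for Phoebe AI bots that passes only if the bot has ever skydived from the battle bus and landed.
; Its AI condition check gets the owner controller's player state, casts it to FortPlayerStateZone and returns that state's HasEverSkydivedFromBusAndLanded result.
; It reads the controller's player state rather than the controlled pawn, which goes unused.
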(fn PerformConditionCheckAI
  (params (OwnerController :type AIController%20Object%20Reference) (ControlledPawn :type Pawn%20Object%20Reference))
  (bind _asfort_player_state_zone (Utilities|Casting|CastToFortPlayerStateZone (Variables|/Script/Engine.Controller|GetPlayerState OwnerController)))
  (return (Class|FortPlayerStateZone|HasEverSkydivedfromBusandLanded _asfort_player_state_zone)))
