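; Behavior tree decorator for Phoebe AI that passes its condition at random, according to a Chance variable.
; The AI condition check treats Chance values up to one as a fraction and larger values as a percentage divided by 100.
; It then passes when a new random float is at or below that normalized value, ignoring the controller and pawn.
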(fn PerformConditionCheckAI
  (params (OwnerController :type AIController%20Object%20Reference) (ControlledPawn :type Pawn%20Object%20Reference))
  (if (<= (Variables|Default|GetChance) 1.0)
    (LocalVariables|SetTestPercent (Variables|Default|GetChance))
    (else
      (LocalVariables|SetTestPercent (/ (Variables|Default|GetChance) 1e+02))))
  (return (<= (Math|Random|RandomFloat) (LocalVariables|GetTestPercent))))
